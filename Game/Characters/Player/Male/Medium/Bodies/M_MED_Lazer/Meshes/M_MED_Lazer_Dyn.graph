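# Physics asset for the M_MED_Lazer medium male player body, driving secondary motion of its collar, jacket flaps, armband and chest strap.
# Spine, pelvis and head capsules form a freely rotating core, with MSTR anchor bodies hung from spine_03 and spine_05.
# Collar and jacket boxes swing within limits from their anchors, while the armband knot branches into two short capsule chains.
# The chest strap swings on one axis from spine_04.
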PhysicsAsset /Game/Characters/Player/Male/Medium/Bodies/M_MED_Lazer/Meshes/M_MED_Lazer_Dyn
bodies {
 dyn_armband_hi_01 [{shapeType=Capsule, length=0.4753, radius=1, center=<1 0 0>, rotation=<rot 90 0 0>}]
 dyn_armband_hi_02 [{shapeType=Capsule, length=0.6976, radius=1, center=<1.1187 0 0>, rotation=<rot 90 0 0>}]
 dyn_armband_knot [{shapeType=Capsule, length=1, radius=1, center=<0.1 -0.0001 0>, rotation=<rot 90 0 0>}]
 dyn_armband_knot_MSTR [{shapeType=Capsule, length=0.505, radius=0.505, center=<-1.3055 0 0>}]
 dyn_armband_low_01 [{shapeType=Capsule, length=0.7121, radius=1, center=<1.4721 0 0>, rotation=<rot 90 0 0>}]
 dyn_armband_low_02 [{shapeType=Capsule, length=1, radius=1, center=<1 0 0>, rotation=<rot 90 0 0>}]
 dyn_chest_strap [{shapeType=Capsule, length=7.083, radius=4.2194, center=<0.0772 -0.0648 0>, rotation=<rot 90 4.0856 44.0865>}]
 dyn_collar_back [{shapeType=Box, extentX=5.1852, extentY=12.2737, extentZ=1.0178, center=<2.3542 0.0002 -0.0002>}]
 dyn_collar_back_MSTR [{shapeType=Capsule, length=0.5, radius=0.5, rotation=<rot 90 0 0>}]
 dyn_collar_bk_l [{shapeType=Box, extentX=5.1852, extentY=12.2737, extentZ=1.0178, center=<2.3542 0.0002 -0.0002>}]
 dyn_collar_bk_l_MSTR [{shapeType=Capsule, length=0.5, radius=0.5, rotation=<rot 90 0 0>}]
 dyn_collar_bk_r [{shapeType=Box, extentX=5.1852, extentY=12.2737, extentZ=1.0178, center=<2.3542 0.0002 -0.0002>}]
 dyn_collar_bk_r_MSTR [{shapeType=Capsule, length=0.5, radius=0.5, rotation=<rot 90 0 0>}]
 dyn_collar_fr_l_01 [{shapeType=Box, extentX=7.5, extentY=10.2645, extentZ=1.0178, center=<4.5706 0 -0.0002>}]
 dyn_collar_fr_l_MSTR [{shapeType=Capsule, length=0.5, radius=0.5, rotation=<rot 90 0 0>}]
 dyn_collar_fr_r_01 [{shapeType=Box, extentX=7.5, extentY=10.2645, extentZ=1.0178, center=<4.5706 0 -0.0002>}]
 dyn_collar_fr_r_MSTR [{shapeType=Capsule, length=0.5, radius=0.5, rotation=<rot 90 0 0>}]
 dyn_collar_l [{shapeType=Box, extentX=5.1852, extentY=12.2737, extentZ=1.0178, center=<2.3542 0.0002 -0.0002>}]
 dyn_collar_l_MSTR [{shapeType=Capsule, length=0.5, radius=0.5, rotation=<rot 90 0 0>}]
 dyn_collar_r [{shapeType=Box, extentX=5.1852, extentY=12.2737, extentZ=1.0178, center=<2.3542 0.0002 -0.0002>}]
 dyn_collar_r_MSTR [{shapeType=Capsule, length=0.5, radius=0.5, rotation=<rot 90 0 0>}]
 dyn_jacket_bk_01 [{shapeType=Box, extentX=10.5199, extentY=12.5949, extentZ=1.0178, center=<5.3093 0 -0.3374>, rotation=<rot -4.3048 0 0>}]
 dyn_jacket_bk_MSTR [{shapeType=Capsule, length=0.5, radius=0.5, rotation=<rot 90 0 0>}]
 dyn_jacket_bk_r_01 [{shapeType=Box, extentX=10.7183, extentY=12.5949, extentZ=4.6537, center=<4.4564 0 0>}]
 dyn_jacket_bk_r_MSTR [{shapeType=Capsule, length=0.5, radius=0.5, rotation=<rot 90 0 0>}]
 dyn_jacket_fr_r_01 [{shapeType=Box, extentX=14.3799, extentY=12.5949, extentZ=1.0178, center=<6.7437 0.0001 0>}]
 dyn_jacket_fr_r_MSTR [{shapeType=Capsule, length=0.5, radius=0.5, rotation=<rot 90 0 0>}]
 dyn_jacket_r_01 [{shapeType=Box, extentX=11.6788, extentY=12.5949, extentZ=1.8366, center=<5.6106 -0.0001 0>}]
 dyn_jacket_r_MSTR [{shapeType=Capsule, length=0.5, radius=0.5, rotation=<rot 90 0 0>}]
 head [{shapeType=Capsule, length=3.6348, radius=7.011, center=<2.6653 1.8932 -0.0001>, rotation=<rot 90 0 0>}]
 pelvis [{shapeType=Capsule, length=2, radius=2}]
 spine_01 [{shapeType=Capsule, length=2, radius=2}]
 spine_02 [{shapeType=Capsule, length=8.698, radius=9.1619, center=<0.0001 0.542 0>}]
 spine_03 [{shapeType=Capsule, length=8.698, radius=9.1619, center=<0.0001 0.542 0>}]
 spine_04 [{shapeType=Capsule, length=1, radius=1}]
 spine_05 [{shapeType=Capsule, length=2, radius=2}]
}
constraints [
 {
  bone1Name=spine_01
  bone2Name=pelvis
  swing1LimitDegrees=45
  swing1Motion=Free
  swing2LimitDegrees=45
  swing2Motion=Free
  twistLimitDegrees=45
  twistMotion=Free
 }
 {
  bone1Name=spine_05
  bone2Name=spine_04
  swing1LimitDegrees=45
  swing1Motion=Free
  swing2LimitDegrees=45
  swing2Motion=Free
  twistLimitDegrees=45
  twistMotion=Free
 }
 {
  bone1Name=dyn_collar_fr_l_MSTR
  bone2Name=spine_05
  swing1LimitDegrees=45
  swing1Motion=Free
  swing2LimitDegrees=45
  swing2Motion=Free
  twistLimitDegrees=45
  twistMotion=Free
 }
 {bone1Name=dyn_collar_fr_l_01, bone2Name=dyn_collar_fr_l_MSTR, swing1LimitDegrees=25, swing2LimitDegrees=25, twistLimitDegrees=75}
 {
  bone1Name=dyn_collar_l_MSTR
  bone2Name=spine_05
  swing1LimitDegrees=45
  swing1Motion=Free
  swing2LimitDegrees=45
  swing2Motion=Free
  twistLimitDegrees=45
  twistMotion=Free
 }
 {bone1Name=dyn_collar_l, bone2Name=dyn_collar_l_MSTR, swing1LimitDegrees=25, swing2LimitDegrees=25, twistLimitDegrees=75}
 {
  bone1Name=dyn_collar_bk_l_MSTR
  bone2Name=spine_05
  swing1LimitDegrees=45
  swing1Motion=Free
  swing2LimitDegrees=45
  swing2Motion=Free
  twistLimitDegrees=45
  twistMotion=Free
 }
 {bone1Name=dyn_collar_bk_l, bone2Name=dyn_collar_bk_l_MSTR, swing1LimitDegrees=75, swing2LimitDegrees=75, twistLimitDegrees=75}
 {
  bone1Name=dyn_collar_back_MSTR
  bone2Name=spine_05
  swing1LimitDegrees=45
  swing1Motion=Free
  swing2LimitDegrees=45
  swing2Motion=Free
  twistLimitDegrees=45
  twistMotion=Free
 }
 {bone1Name=dyn_collar_back, bone2Name=dyn_collar_back_MSTR, swing1LimitDegrees=75, swing2LimitDegrees=75, twistLimitDegrees=75}
 {
  bone1Name=dyn_collar_bk_r_MSTR
  bone2Name=spine_05
  swing1LimitDegrees=45
  swing1Motion=Free
  swing2LimitDegrees=45
  swing2Motion=Free
  twistLimitDegrees=45
  twistMotion=Free
 }
 {bone1Name=dyn_collar_bk_r, bone2Name=dyn_collar_bk_r_MSTR, swing1LimitDegrees=75, swing2LimitDegrees=75, twistLimitDegrees=75}
 {
  bone1Name=dyn_collar_r_MSTR
  bone2Name=spine_05
  swing1LimitDegrees=45
  swing1Motion=Free
  swing2LimitDegrees=45
  swing2Motion=Free
  twistLimitDegrees=45
  twistMotion=Free
 }
 {bone1Name=dyn_collar_r, bone2Name=dyn_collar_r_MSTR, swing1LimitDegrees=25, swing2LimitDegrees=25, twistLimitDegrees=75}
 {
  bone1Name=dyn_collar_fr_r_MSTR
  bone2Name=spine_05
  swing1LimitDegrees=45
  swing1Motion=Free
  swing2LimitDegrees=45
  swing2Motion=Free
  twistLimitDegrees=45
  twistMotion=Free
 }
 {bone1Name=dyn_collar_fr_r_01, bone2Name=dyn_collar_fr_r_MSTR, swing1LimitDegrees=25, swing2LimitDegrees=25, twistLimitDegrees=75}
 {
  bone1Name=dyn_jacket_fr_r_MSTR
  bone2Name=spine_03
  swing1LimitDegrees=45
  swing1Motion=Free
  swing2LimitDegrees=45
  swing2Motion=Free
  twistLimitDegrees=45
  twistMotion=Free
 }
 {bone1Name=dyn_jacket_fr_r_01, bone2Name=dyn_jacket_fr_r_MSTR, swing1LimitDegrees=45, swing2LimitDegrees=45, twistLimitDegrees=5}
 {
  bone1Name=dyn_jacket_r_MSTR
  bone2Name=spine_03
  swing1LimitDegrees=45
  swing1Motion=Free
  swing2LimitDegrees=45
  swing2Motion=Free
  twistLimitDegrees=45
  twistMotion=Free
 }
 {bone1Name=dyn_jacket_r_01, bone2Name=dyn_jacket_r_MSTR, swing1LimitDegrees=45, swing2LimitDegrees=45, twistLimitDegrees=5}
 {
  bone1Name=dyn_jacket_bk_MSTR
  bone2Name=spine_03
  swing1LimitDegrees=45
  swing1Motion=Free
  swing2LimitDegrees=45
  swing2Motion=Free
  twistLimitDegrees=45
  twistMotion=Free
 }
 {bone1Name=dyn_jacket_bk_01, bone2Name=dyn_jacket_bk_MSTR, swing1LimitDegrees=45, swing2LimitDegrees=45, twistLimitDegrees=5}
 {
  bone1Name=dyn_jacket_bk_r_MSTR
  bone2Name=spine_03
  swing1LimitDegrees=45
  swing1Motion=Free
  swing2LimitDegrees=45
  swing2Motion=Free
  twistLimitDegrees=45
  twistMotion=Free
 }
 {bone1Name=dyn_jacket_bk_r_01, bone2Name=dyn_jacket_bk_r_MSTR, swing1LimitDegrees=45, swing2LimitDegrees=45, twistLimitDegrees=5}
 {
  bone1Name=head
  bone2Name=spine_05
  swing1LimitDegrees=45
  swing1Motion=Free
  swing2LimitDegrees=45
  swing2Motion=Free
  twistLimitDegrees=45
  twistMotion=Free
 }
 {
  bone1Name=spine_02
  bone2Name=spine_01
  swing1LimitDegrees=45
  swing1Motion=Free
  swing2LimitDegrees=45
  swing2Motion=Free
  twistLimitDegrees=45
  twistMotion=Free
 }
 {
  bone1Name=spine_03
  bone2Name=spine_02
  swing1LimitDegrees=45
  swing1Motion=Free
  swing2LimitDegrees=45
  swing2Motion=Free
  twistLimitDegrees=45
  twistMotion=Free
 }
 {
  bone1Name=dyn_armband_knot_MSTR
  bone2Name=spine_05
  swing1LimitDegrees=45
  swing1Motion=Free
  swing2LimitDegrees=45
  swing2Motion=Free
  twistLimitDegrees=45
  twistMotion=Free
 }
 {bone1Name=dyn_armband_knot, bone2Name=dyn_armband_knot_MSTR, swing1LimitDegrees=45, swing2LimitDegrees=45, twistLimitDegrees=5}
 {bone1Name=dyn_armband_low_01, bone2Name=dyn_armband_knot, swing1LimitDegrees=45, swing2LimitDegrees=45, twistLimitDegrees=5}
 {bone1Name=dyn_armband_low_02, bone2Name=dyn_armband_low_01, swing1LimitDegrees=45, swing2LimitDegrees=45, twistLimitDegrees=5}
 {bone1Name=dyn_armband_hi_01, bone2Name=dyn_armband_knot, swing1LimitDegrees=45, swing2LimitDegrees=45, twistLimitDegrees=5}
 {bone1Name=dyn_armband_hi_02, bone2Name=dyn_armband_hi_01, swing1LimitDegrees=45, swing2LimitDegrees=45, twistLimitDegrees=5}
 {
  bone1Name=dyn_chest_strap
  bone2Name=spine_04
  swing1LimitDegrees=60
  swing2LimitDegrees=60
  swing2Motion=Locked
  twistLimitDegrees=5
  twistMotion=Locked
 }
 {
  bone1Name=spine_04
  bone2Name=spine_03
  swing1LimitDegrees=45
  swing1Motion=Free
  swing2LimitDegrees=45
  swing2Motion=Free
  twistLimitDegrees=45
  twistMotion=Free
 }
]
