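# Animation Blueprint for the M_MED_Lazer player body that copies the parent mesh pose and adds secondary motion to shoulder pads and an armband knot.
# Upper-arm pose drivers pose the pads, then constraints adjust the right pad and knot before RigidBody simulation.
# Targeting events toggle an interpolated alpha for additive offsets on two right-pad bones, applied before simulation.
# Each update computes arm_up_l/r from upper-arm yaw, though the AnimGraph never reads them.
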AnimBlueprint /Game/Characters/Player/Male/Medium/Bodies/M_MED_Lazer/Meshes/M_MED_Lazer_AnimBP
target_skeleton=&/M_MED_Lazer_Skeleton
graphs [
 AnimGraph {nodes [
   n1 = Pose Driver - Source: upperarm_r!(SourcePose:Pose Link Structure=Copy Pose From Mesh() {bUseAttachedParent=true}) -> (Pose:Pose Link Structure) {
    is=title AnimGraphNode_PoseDriver
    axisLength=-20
    poseAsset=&/Shoulderpad_PoseDriver_PoseAsset
    onlyDriveBones [{boneName=dyn_shoulderpad_r_01}, {boneName=dyn_shoulderpad_r_02}, {boneName=dyn_shoulderpad_r_03}]
    poseTargets [
     {
      bApplyCustomCurve=false
      bIsHidden=false
      distanceMethod=DefaultMethod
      drivenName=pickaxe
      functionType=DefaultFunction
      targetScale=0.45
      boneTransforms [{targetRotation=<rot -72.819 26.9102 -26.9579>, targetTranslation=<-15.5097 0 -0.0005>}]
      customCurve {defaultValue=3.4028234663852886e+38, postInfinityExtrap=RCCE_Constant, preInfinityExtrap=RCCE_Constant, keys []}
      targetRotation=<rot 0 0 0>
     }
     {
      bApplyCustomCurve=false
      bIsHidden=false
      distanceMethod=DefaultMethod
      drivenName=rifle_fwd
      functionType=DefaultFunction
      targetScale=1
      boneTransforms [{targetRotation=<rot -47.0553 101.4345 -53.4248>, targetTranslation=<-15.5094 -0.0003 0.0021>}]
      customCurve {defaultValue=3.4028234663852886e+38, postInfinityExtrap=RCCE_Constant, preInfinityExtrap=RCCE_Constant, keys []}
      targetRotation=<rot 0 0 0>
     }
     {
      bApplyCustomCurve=false
      bIsHidden=false
      distanceMethod=DefaultMethod
      drivenName=rifle down
      functionType=DefaultFunction
      targetScale=1
      boneTransforms [{targetRotation=<rot -65.4825 116.8243 -64.5283>, targetTranslation=<-15.5159 0.0068 -0.0028>}]
      customCurve {defaultValue=3.4028234663852886e+38, postInfinityExtrap=RCCE_Constant, preInfinityExtrap=RCCE_Constant, keys []}
      targetRotation=<rot 0 0 0>
     }
     {
      bApplyCustomCurve=false
      bIsHidden=false
      distanceMethod=DefaultMethod
      drivenName=rifle up
      functionType=DefaultFunction
      targetScale=0.55
      boneTransforms [{targetRotation=<rot -33.7135 111.062 -66.4336>, targetTranslation=<-15.5096 0.0001 0.002>}]
      customCurve {defaultValue=3.4028234663852886e+38, postInfinityExtrap=RCCE_Constant, preInfinityExtrap=RCCE_Constant, keys []}
      targetRotation=<rot 0 0 0>
     }
     {
      bApplyCustomCurve=false
      bIsHidden=false
      distanceMethod=DefaultMethod
      drivenName=pistol fwd
      functionType=DefaultFunction
      targetScale=0.7
      boneTransforms [{targetRotation=<rot -8.3178 104.8664 -59.5298>, targetTranslation=<-15.509 -0.0002 0.0025>}]
      customCurve {defaultValue=3.4028234663852886e+38, postInfinityExtrap=RCCE_Constant, preInfinityExtrap=RCCE_Constant, keys []}
      targetRotation=<rot 0 0 0>
     }
     {
      bApplyCustomCurve=false
      bIsHidden=false
      distanceMethod=DefaultMethod
      drivenName=pistol down
      functionType=DefaultFunction
      targetScale=0.7
      boneTransforms [{targetRotation=<rot -44.5194 113.5295 -48.8246>, targetTranslation=<-15.5091 -0.0004 0.0018>}]
      customCurve {defaultValue=3.4028234663852886e+38, postInfinityExtrap=RCCE_Constant, preInfinityExtrap=RCCE_Constant, keys []}
      targetRotation=<rot 0 0 0>
     }
     {
      bApplyCustomCurve=false
      bIsHidden=false
      distanceMethod=DefaultMethod
      drivenName=pistol up
      functionType=DefaultFunction
      targetScale=0.5
      boneTransforms [{targetRotation=<rot 34.7596 88.3742 -91.0759>, targetTranslation=<-15.5093 -0.0001 0.0011>}]
      customCurve {defaultValue=3.4028234663852886e+38, postInfinityExtrap=RCCE_Constant, preInfinityExtrap=RCCE_Constant, keys []}
      targetRotation=<rot 0 0 0>
     }
     {
      bApplyCustomCurve=false
      bIsHidden=false
      distanceMethod=DefaultMethod
      drivenName=rocket fwd
      functionType=DefaultFunction
      targetScale=0.58
      boneTransforms [{targetRotation=<rot -44.3089 91.3127 -61.5273>, targetTranslation=<-15.5096 -0.0001 0.001>}]
      customCurve {defaultValue=3.4028234663852886e+38, postInfinityExtrap=RCCE_Constant, preInfinityExtrap=RCCE_Constant, keys []}
      targetRotation=<rot 0 0 0>
     }
     {
      bApplyCustomCurve=false
      bIsHidden=false
      distanceMethod=DefaultMethod
      drivenName=rocket down
      functionType=DefaultFunction
      targetScale=0.53
      boneTransforms [{targetRotation=<rot -69.9453 -32.8912 71.3344>, targetTranslation=<-15.5223 0.0174 -0.0182>}]
      customCurve {defaultValue=3.4028234663852886e+38, postInfinityExtrap=RCCE_Constant, preInfinityExtrap=RCCE_Constant, keys []}
      targetRotation=<rot 0 0 0>
     }
     {
      bApplyCustomCurve=false
      bIsHidden=false
      distanceMethod=DefaultMethod
      drivenName=rocket up
      functionType=DefaultFunction
      targetScale=0.66
      boneTransforms [{targetRotation=<rot 7.7473 89.9061 -78.4301>, targetTranslation=<-15.5094 -0.002 0.0097>}]
      customCurve {defaultValue=3.4028234663852886e+38, postInfinityExtrap=RCCE_Constant, preInfinityExtrap=RCCE_Constant, keys []}
      targetRotation=<rot 0 0 0>
     }
     {
      bApplyCustomCurve=false
      bIsHidden=false
      distanceMethod=DefaultMethod
      drivenName=minigun fwd
      functionType=DefaultFunction
      targetScale=0.43
      boneTransforms [{targetRotation=<rot -59.645 -14.7379 62.5845>, targetTranslation=<-15.5094 -0.0003 0.0017>}]
      customCurve {defaultValue=3.4028234663852886e+38, postInfinityExtrap=RCCE_Constant, preInfinityExtrap=RCCE_Constant, keys []}
      targetRotation=<rot 0 0 0>
     }
     {
      bApplyCustomCurve=false
      bIsHidden=false
      distanceMethod=DefaultMethod
      drivenName=freefall
      functionType=DefaultFunction
      targetScale=0.8
      boneTransforms [{targetRotation=<rot 5.3968 1.5427 -69.143>, targetTranslation=<-15.5105 -0.0002 -0.0003>}]
      customCurve {defaultValue=3.4028234663852886e+38, postInfinityExtrap=RCCE_Constant, preInfinityExtrap=RCCE_Constant, keys []}
      targetRotation=<rot 0 0 0>
     }
     {
      bApplyCustomCurve=false
      bIsHidden=false
      distanceMethod=DefaultMethod
      drivenName=glider
      functionType=DefaultFunction
      targetScale=0.8
      boneTransforms [{targetRotation=<rot 33.7743 -6.7914 -76.673>, targetTranslation=<-15.5104 -0.0002 -0.0001>}]
      customCurve {defaultValue=3.4028234663852886e+38, postInfinityExtrap=RCCE_Constant, preInfinityExtrap=RCCE_Constant, keys []}
      targetRotation=<rot 0 0 0>
     }
    ]
    rBFParams {radius=44}
    sourceBones [{boneName=upperarm_r}]
   }
   n2 = Pose Driver - Source: upperarm_l!(SourcePose:Pose Link Structure=@n1) -> (Pose:Pose Link Structure) {
    is=title AnimGraphNode_PoseDriver
    poseAsset=&Shoulderpad_PoseDriver_PoseAsset
    onlyDriveBones [{boneName=dyn_shoulderpad_l_01}]
    poseTargets [
     {
      bApplyCustomCurve=false
      bIsHidden=false
      distanceMethod=DefaultMethod
      drivenName=pickaxe
      functionType=DefaultFunction
      targetScale=0.3263
      boneTransforms [{targetRotation=<rot -50.7937 37.3184 6.585>, targetTranslation=<15.5076 0.0002 -0.0015>}]
      customCurve {defaultValue=3.4028234663852886e+38, postInfinityExtrap=RCCE_Constant, preInfinityExtrap=RCCE_Constant, keys []}
      targetRotation=<rot 0 0 0>
     }
     {
      bApplyCustomCurve=false
      bIsHidden=false
      distanceMethod=DefaultMethod
      drivenName=rifle_fwd
      functionType=DefaultFunction
      targetScale=0.179
      boneTransforms [{targetRotation=<rot -35.3872 64.8154 -82.9358>, targetTranslation=<15.5079 -0.0002 0.0014>}]
      customCurve {defaultValue=3.4028234663852886e+38, postInfinityExtrap=RCCE_Constant, preInfinityExtrap=RCCE_Constant, keys []}
      targetRotation=<rot 0 0 0>
     }
     {
      bApplyCustomCurve=false
      bIsHidden=false
      distanceMethod=DefaultMethod
      drivenName=rifle down
      functionType=DefaultFunction
      targetScale=0.3295
      boneTransforms [{targetRotation=<rot -64.1607 65.8079 -40.3078>, targetTranslation=<15.5127 0.0087 -0.0097>}]
      customCurve {defaultValue=3.4028234663852886e+38, postInfinityExtrap=RCCE_Constant, preInfinityExtrap=RCCE_Constant, keys []}
      targetRotation=<rot 0 0 0>
     }
     {
      bApplyCustomCurve=false
      bIsHidden=false
      distanceMethod=DefaultMethod
      drivenName=rifle up
      functionType=DefaultFunction
      targetScale=0.5
      boneTransforms [{targetRotation=<rot -5.2406 47.1265 -85.1312>, targetTranslation=<15.5083 0.0001 0.0003>}]
      customCurve {defaultValue=3.4028234663852886e+38, postInfinityExtrap=RCCE_Constant, preInfinityExtrap=RCCE_Constant, keys []}
      targetRotation=<rot 0 0 0>
     }
     {
      bApplyCustomCurve=false
      bIsHidden=false
      distanceMethod=DefaultMethod
      drivenName=pistol fwd
      functionType=DefaultFunction
      targetScale=0.179
      boneTransforms [{targetRotation=<rot -31.6194 70.4927 -60.4477>, targetTranslation=<15.5082 -0.0003 0.0025>}]
      customCurve {defaultValue=3.4028234663852886e+38, postInfinityExtrap=RCCE_Constant, preInfinityExtrap=RCCE_Constant, keys []}
      targetRotation=<rot 0 0 0>
     }
     {
      bApplyCustomCurve=false
      bIsHidden=false
      distanceMethod=DefaultMethod
      drivenName=pistol up
      functionType=DefaultFunction
      targetScale=0.4683
      boneTransforms [{targetRotation=<rot 8.6219 54.7727 -96.1121>, targetTranslation=<15.508 0 0.0003>}]
      customCurve {defaultValue=3.4028234663852886e+38, postInfinityExtrap=RCCE_Constant, preInfinityExtrap=RCCE_Constant, keys []}
      targetRotation=<rot 0 0 0>
     }
     {
      bApplyCustomCurve=false
      bIsHidden=false
      distanceMethod=DefaultMethod
      drivenName=rocket fwd
      functionType=DefaultFunction
      targetScale=0.67
      boneTransforms [{targetRotation=<rot -19.8367 86.1295 -81.5179>, targetTranslation=<15.5078 0 0.0008>}]
      customCurve {defaultValue=3.4028234663852886e+38, postInfinityExtrap=RCCE_Constant, preInfinityExtrap=RCCE_Constant, keys []}
      targetRotation=<rot 0 0 0>
     }
     {
      bApplyCustomCurve=false
      bIsHidden=false
      distanceMethod=DefaultMethod
      drivenName=rocket up
      functionType=DefaultFunction
      targetScale=0.59
      boneTransforms [{targetRotation=<rot 60.4608 45.6429 -56.9818>, targetTranslation=<15.5109 -0.0022 0.0065>}]
      customCurve {defaultValue=3.4028234663852886e+38, postInfinityExtrap=RCCE_Constant, preInfinityExtrap=RCCE_Constant, keys []}
      targetRotation=<rot 0 0 0>
     }
     {
      bApplyCustomCurve=false
      bIsHidden=false
      distanceMethod=DefaultMethod
      drivenName=minigun fwd
      functionType=DefaultFunction
      targetScale=0.3897
      boneTransforms [{targetRotation=<rot -48.4942 63.1202 -22.8392>, targetTranslation=<15.5075 -0.0003 -0.0011>}]
      customCurve {defaultValue=3.4028234663852886e+38, postInfinityExtrap=RCCE_Constant, preInfinityExtrap=RCCE_Constant, keys []}
      targetRotation=<rot 0 0 0>
     }
     {
      bApplyCustomCurve=false
      bIsHidden=false
      distanceMethod=DefaultMethod
      drivenName=minigun down
      functionType=DefaultFunction
      targetScale=0.3263
      boneTransforms [{targetRotation=<rot -59.6182 25.6392 -19.4968>, targetTranslation=<15.521 0.0105 -0.0592>}]
      customCurve {defaultValue=3.4028234663852886e+38, postInfinityExtrap=RCCE_Constant, preInfinityExtrap=RCCE_Constant, keys []}
      targetRotation=<rot 0 0 0>
     }
     {
      bApplyCustomCurve=false
      bIsHidden=false
      distanceMethod=DefaultMethod
      drivenName=freefall
      functionType=DefaultFunction
      targetScale=1
      boneTransforms [{targetRotation=<rot -4.067 2.9236 -70.6093>, targetTranslation=<15.5076 0 0.0004>}]
      customCurve {defaultValue=3.4028234663852886e+38, postInfinityExtrap=RCCE_Constant, preInfinityExtrap=RCCE_Constant, keys []}
      targetRotation=<rot 0 0 0>
     }
     {
      bApplyCustomCurve=false
      bIsHidden=false
      distanceMethod=DefaultMethod
      drivenName=glider
      functionType=DefaultFunction
      targetScale=0.54
      boneTransforms [{targetRotation=<rot 49.4484 57.8648 -63.0965>, targetTranslation=<15.5076 0.0006 0.0004>}]
      customCurve {defaultValue=3.4028234663852886e+38, postInfinityExtrap=RCCE_Constant, preInfinityExtrap=RCCE_Constant, keys []}
      targetRotation=<rot 0 0 0>
     }
    ]
    rBFParams {radius=57.9}
    sourceBones [{boneName=upperarm_l}]
   }
   n3 = "Constraint - dyn_shoulderpad_r_02 (R)"(ComponentPose=Local To Component(LocalPose=@n2), Alpha=0.3, ConstraintWeights_0=1) {
    alphaBoolBlend {blendOption=Linear}
    boneToModify {boneName=dyn_shoulderpad_r_02}
    constraintSetup [{transformType=Rotation, targetBone {boneName=upperarm_twist_01_r}}]
    constraintWeights [1]
   }
   n4 = "Constraint - dyn_shoulderpad_r_03 (R)"(ComponentPose=@n3, Alpha=0.5, ConstraintWeights_0=1) {
    alphaBoolBlend {blendOption=Linear}
    boneToModify {boneName=dyn_shoulderpad_r_03}
    constraintSetup [{transformType=Rotation, targetBone {boneName=upperarm_twist_01_r}}]
    constraintWeights [1]
   }
   n5 = "Constraint - dyn_armband_knot_MSTR (PPPPPPP)"!(
    ComponentPose=@n4
    Alpha=1
    ConstraintWeights_0=0.0927
    ConstraintWeights_1:float=0.287
    ConstraintWeights_2:float=0
    ConstraintWeights_3:float=0.3412
    ConstraintWeights_4:float=0.0052
    ConstraintWeights_5:float=0.2738
    ConstraintWeights_6:float=0.0002
   ) -> (Pose) {
    alphaBoolBlend {blendOption=Linear}
    boneToModify {boneName=dyn_armband_knot_MSTR}
    constraintSetup [
     {targetBone {boneName=clavicle_l}}
     {targetBone {boneName=upperarm_twist_02_l}}
     {targetBone {boneName=spine_04}}
     {targetBone {boneName=upperarm_twist_01_l}}
     {targetBone {boneName=lowerarm_l}}
     {targetBone {boneName=upperarm_l}}
     {targetBone {boneName=lowerarm_twist_02_l}}
    ]
    constraintWeights [1,1,1,1,1,1,1]
   }
   n6 = Get Targeting Alpha() -> (Targeting Alpha:double)
   n7 = "Transform (Modify) Bone - Bone: dyn_shoulderpad_r_02"(ComponentPose=@n5, Alpha=@n6, Translation="0.028,0.000,-0.742", Rotation="19.790,0.000,0.000", Scale="1,1,1") {
    rotationMode=BMM_Additive
    rotationSpace=BCS_BoneSpace
    translationMode=BMM_Additive
    translationSpace=BCS_BoneSpace
    alphaBoolBlend {blendOption=Linear}
    alphaScaleBiasClamp {bInterpResult=true}
    boneToModify {boneName=dyn_shoulderpad_r_02}
    rotation {pitch=19.79}
    translation {x=0.028, z=-0.742}
   }
   n8 = "Transform (Modify) Bone - Bone: dyn_shoulderpad_r_03"(ComponentPose=@n7, Alpha=@n6, Translation="-0.416,-0.000,-0.579", Rotation="17.172,0.000,0.000", Scale="1,1,1") {
    rotationMode=BMM_Additive
    rotationSpace=BCS_BoneSpace
    translationMode=BMM_Additive
    translationSpace=BCS_BoneSpace
    alphaBoolBlend {blendOption=Linear}
    alphaScaleBiasClamp {bInterpResult=true}
    boneToModify {boneName=dyn_shoulderpad_r_03}
    rotation {pitch=17.172}
    translation {x=-0.416, z=-0.579}
   }
   n9 = RigidBody(ComponentPose=@n8) {
    bClampLinearTranslationLimitToRefPose=true
    bOverrideWorldGravity=true
    overridePhysicsAsset=&/M_MED_Lazer_Dyn
    simulationSpace=BaseBoneSpace
    alphaBoolBlend {blendOption=Linear}
    baseBoneRef {boneName=spine_02}
    componentAppliedLinearAccClamp=<3000 3000 3000>
    componentLinearAccScale=<0.3 0.35 0.35>
    componentLinearVelScale=<0.3 0.35 0.35>
    overrideWorldGravity {z=-980}
   }
   Output Pose(Result=Component To Local(ComponentPose=@n9)) {layerGroup=None}
  ]}
 EventGraph {nodes [
   Event BlueprintUpdateAnimation()
   n1 = GetOwnerSkeletalMeshComponent()
   n2 = BreakRotator(InRot=BreakTransform(InTransform=GetDeltaTransformFromRefPose(self=@n1, BoneName=upperarm_r, BaseName=spine_05)).Rotation)
   > Set arm_up_r(arm_up_r:double=MapRangeClamped(Value="Absolute (Float)"(A=@n2.Yaw), InRangeA=40, InRangeB=90)) -> (Output_Get:double)
   n3 = GetOwnerSkeletalMeshComponent()
   n4 = BreakRotator(InRot=BreakTransform(InTransform=GetDeltaTransformFromRefPose(self=@n3, BoneName=upperarm_l, BaseName=spine_05)).Rotation)
   > Set arm_up_l(arm_up_l:double=MapRangeClamped(Value="Absolute (Float)"(A=@n4.Yaw), InRangeA=40, InRangeB=90)) -> (Output_Get:double)
   Event OnBeginTargeting() {is=title K2Node_Event}
   > Set Targeting Alpha(Targeting Alpha:double=1) -> (Output_Get:double)
   Event OnEndTargeting() {is=title K2Node_Event}
   > Set Targeting Alpha(Targeting Alpha=0)
  ]}
]
variables [
 arm_up_l {type=number}
 arm_up_r {type=number}
 Targeting Alpha {type=number}
]
