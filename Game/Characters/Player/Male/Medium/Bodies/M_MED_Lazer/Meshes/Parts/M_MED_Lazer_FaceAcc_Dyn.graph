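# Physics asset for the medium male Lazer player body's face accessory, defining dynamic simulation of its hair and hat bones.
# Four two-segment hair tails hang from a hair base, which, like the hat, attaches to the head with limited swing and tight twist.
# Capsules on head, upper spine and both clavicles form the body shapes around these chains, linked to the spine by free-rotation constraints.
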PhysicsAsset /Game/Characters/Player/Male/Medium/Bodies/M_MED_Lazer/Meshes/Parts/M_MED_Lazer_FaceAcc_Dyn
bodies {
 clavicle_l [{shapeType=Capsule, length=10.5662, radius=8.2669, center=<10.0993 -0.8696 0.1977>, rotation=<rot 84.9998 0 -5>}]
 clavicle_r [{shapeType=Capsule, length=5, radius=5, rotation=<rot 90 0 0>}]
 dyn_hair_base [{shapeType=Capsule, length=0.9102, radius=2.5863, center=<0.5 0 0>, rotation=<rot 90 0 0>}]
 dyn_hair_tail_A_01 [{shapeType=Capsule, length=3.7692, radius=1, center=<3 0 0>, rotation=<rot 90 0 0>}]
 dyn_hair_tail_A_02 [{shapeType=Capsule, length=5.3776, radius=1, center=<3 0 0>, rotation=<rot 90 0 0>}]
 dyn_hair_tail_B_01 [{shapeType=Capsule, length=4.1442, radius=1, center=<3 0 0>, rotation=<rot 90 0 0>}]
 dyn_hair_tail_B_02 [{shapeType=Capsule, length=5.2144, radius=1, center=<3 0 0>, rotation=<rot 90 0 0>}]
 dyn_hair_tail_C_01 [{shapeType=Capsule, length=2.4022, radius=1, center=<3 0 0>, rotation=<rot 90 0 0>}]
 dyn_hair_tail_C_02 [{shapeType=Capsule, length=7.6185, radius=1, center=<3 0 0>, rotation=<rot 90 0 0>}]
 dyn_hair_tail_D_01 [{shapeType=Capsule, length=1.8481, radius=1.1587, center=<3 0 0>, rotation=<rot 90 0 0>}]
 dyn_hair_tail_D_02 [{shapeType=Capsule, length=6.6841, radius=1, center=<3 0 0>, rotation=<rot 90 0 0>}]
 dyn_hat [{shapeType=Capsule, length=2.7345, radius=1.4843, center=<2 0 0>, rotation=<rot 90 0 0>}]
 head [{shapeType=Capsule, length=5, radius=5, rotation=<rot 90 0 0>}]
 spine_05 [
  {shapeType=Capsule, length=7.3667, radius=12.6757, center=<2.6233 0.5991 0>}
  {shapeType=Capsule, length=9.4168, radius=3.793, center=<9.4222 1.0131 0>}
  {shapeType=Capsule, length=8.3767, radius=1.0893, center=<9.0555 2.6982 0.0001>}
 ]
}
constraints [
 {bone1Name=dyn_hair_tail_A_01, bone2Name=dyn_hair_base, swing1LimitDegrees=60, swing2LimitDegrees=60, twistLimitDegrees=5}
 {bone1Name=dyn_hair_tail_A_02, bone2Name=dyn_hair_tail_A_01, swing1LimitDegrees=60, swing2LimitDegrees=60, twistLimitDegrees=5}
 {bone1Name=dyn_hair_tail_B_01, bone2Name=dyn_hair_base, swing1LimitDegrees=60, swing2LimitDegrees=60, twistLimitDegrees=5}
 {bone1Name=dyn_hair_tail_B_02, bone2Name=dyn_hair_tail_B_01, swing1LimitDegrees=60, swing2LimitDegrees=60, twistLimitDegrees=5}
 {bone1Name=dyn_hair_tail_C_01, bone2Name=dyn_hair_base, swing1LimitDegrees=60, swing2LimitDegrees=60, twistLimitDegrees=5}
 {bone1Name=dyn_hair_tail_C_02, bone2Name=dyn_hair_tail_C_01, swing1LimitDegrees=60, swing2LimitDegrees=60, twistLimitDegrees=5}
 {bone1Name=dyn_hair_tail_D_01, bone2Name=dyn_hair_base, swing1LimitDegrees=60, swing2LimitDegrees=60, twistLimitDegrees=5}
 {bone1Name=dyn_hair_tail_D_02, bone2Name=dyn_hair_tail_D_01, swing1LimitDegrees=60, swing2LimitDegrees=60, twistLimitDegrees=5}
 {bone1Name=dyn_hat, bone2Name=head, swing1LimitDegrees=45, swing2LimitDegrees=45, twistLimitDegrees=5}
 {bone1Name=dyn_hair_base, bone2Name=head, swing1LimitDegrees=45, swing2LimitDegrees=45, twistLimitDegrees=5}
 {
  bone1Name=head
  bone2Name=spine_05
  swing1LimitDegrees=45
  swing1Motion=Free
  swing2LimitDegrees=45
  swing2Motion=Free
  twistLimitDegrees=45
  twistMotion=Free
 }
 {
  bone1Name=clavicle_l
  bone2Name=spine_05
  swing1LimitDegrees=45
  swing1Motion=Free
  swing2LimitDegrees=45
  swing2Motion=Free
  twistLimitDegrees=45
  twistMotion=Free
 }
 {
  bone1Name=clavicle_r
  bone2Name=spine_05
  swing1LimitDegrees=45
  swing1Motion=Free
  swing2LimitDegrees=45
  swing2Motion=Free
  twistLimitDegrees=45
  twistMotion=Free
 }
]
